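# Additive UI material for the center of the Gauntlet Guns reticle in MurkyVole.
# SDF box outlines cut by crosses form corner pips and a center ring, alongside radially mirrored polygon triangles offset by TriangleSpread.
# A four-way mirrored pip texture moves with the remapped Spread parameter and is added to the outer layer, which OuterPipOpacity scales.
# TargetInRangeSwitch changes that layer from green to red and also drives the corner pips' scale.
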Material /MurkyVoleGameplay/Gameplay/GauntletGuns/UI/M_UI_MurkyVole_GauntletGuns_Reticle_Center
graph {
 properties {
  bAutomaticallySetUsageInEditor=false
  blendMode=BLEND_Additive
  materialDomain=MD_UI
  pixelDepthOffsetMode=PDOM_Legacy
  usedShadingModels=Default Lit
 }
 nodes [
  n1 = ComponentMask(None=TextureCoordinate()) {g=true}
  n2 = NamedRerouteUsage()
  n3 = Multiply(
   A=LinearInterpolate(
    A=Constant3Vector() {constant=<rgba 0.0056 0.8879 0.1812 1>}
    B=Constant3Vector() {constant=<rgba 0.0039 0.9922 0.0667 1>}
    Alpha=Multiply(A=@n1) {constB=3}
   )
   B=Ceil(None=Saturate(None=@n2))
  )
  n4 = Multiply(A=Constant3Vector() {constant=<rgba 1 0.0023 0 1>}, B=NamedRerouteUsage())
  n5 = Floor(None=Add(A=ScalarParameter() {defaultValue=0.25, parameterName=TargetInRangeSwitch}) {constB=0.25})
  n6 = TextureCoordinate()
  n7 = Constant() {r=0.125}
  n8 = MaterialFunctionCall(
   UVs=MaterialFunctionCall!(
    UVs=MaterialFunctionCall(UVs=@n6, "Rotation Angle (0-1)"=@n7) {materialFunction=&CustomRotator}.Rotated Values
    Sides=Constant() {r=4}
   ) {materialFunction=/FortUILibrary/MaterialFunctions/Basics/MF_UI_UVRadialMirror}.UVs
   Texture Scale=Constant() {r=0.125}
  ) {materialFunction=&ScaleUVsByCenter}
  n9 = ComponentMask(None=@n8.UVs) {g=true}
  n10 = MaterialFunctionCall(
   Input=ScalarParameter() {defaultValue=2.8, parameterName=Spread}
   Input Low=Constant() {r=2.8}
   Input High=Constant() {r=16}
   Target Low=Constant() {r=-3}
   Target High=Constant() {r=-1.8}
  ) {materialFunction=&RemapValueRange}
  n11 = TextureSample(UVs=AppendVector(A=ComponentMask(None=@n8.UVs) {r=true}, B=Add(A=@n9, B=@n10.Result))) {
   samplerSource=SSM_Clamp_WorldGroupSettings
   samplerType=SAMPLERTYPE_Grayscale
   texture=/MurkyVoleGameplay/Gameplay/GauntletGuns/UI/V2/T_WM_CenterPip
  }
  n12 = Constant() {r=0.125}
  Add(
   A=Multiply(A=LinearInterpolate(A=@n3, B=@n4, Alpha=@n5), B=ScalarParameter() {defaultValue=1, parameterName=OuterPipOpacity})
   B=Max(
    A=MaterialFunctionCall(DistanceField=@n11.R, Threshold=@n12) {materialFunction=&MF_UI_AntiAliasing}.Outer
    B=NamedRerouteUsage()
   )
  )
  Add(A=TextureCoordinate(), B=Multiply(A=NamedRerouteUsage(), B=Constant2Vector() {g=0.01, r=0.005}) {constB=0.005})
  Constant() {r=0.01}
  n13 = MaterialFunctionCall(
   UVs=TextureCoordinate()
   Texture Scale=LinearInterpolate(Alpha=ScalarParameter() {defaultValue=0.25, parameterName=TargetInRangeSwitch}) {constA=1.1, constB=0.9}
  ) {materialFunction=&ScaleUVsByCenter}
  n14 = MaterialFunctionCall(DistanceField=MaterialFunctionCall(UVs=@n13.UVs, Size=Constant2Vector() {g=0.28, r=0.29}, OutlineThickness=Constant() {r=0.007}, OutlineAlignment=Constant() {r=-1}) {materialFunction=&MF_UI_Box}.SDFOutline) {materialFunction=&MF_UI_AntiAliasing}
  n15 = Max(
   A=MaterialFunctionCall(UVs=@n13.UVs, Size=Constant2Vector() {g=2, r=0.22}) {materialFunction=&MF_UI_Box}.Fill
   B=MaterialFunctionCall(UVs=@n13.UVs, Size=Constant2Vector() {g=0.24, r=2}) {materialFunction=&MF_UI_Box}.Fill
  )
  Divide(A=@n14.Inner, B=OneMinus(None=@n15))
  n16 = MaterialFunctionCall(DistanceField=MaterialFunctionCall(Size=Constant2Vector() {g=0.033, r=0.033}, OutlineThickness=Constant() {r=0.005}, OutlineAlignment=Constant() {r=-1}) {materialFunction=&MF_UI_Box}.SDFOutline) {materialFunction=&MF_UI_AntiAliasing}
  n17 = Max(
   A=MaterialFunctionCall(Size=Constant2Vector() {g=2, r=0.006}) {materialFunction=&MF_UI_Box}.Fill
   B=MaterialFunctionCall(Size=Constant2Vector() {g=0.006, r=2}) {materialFunction=&MF_UI_Box}.Fill
  )
  Divide(A=@n16.Inner, B=OneMinus(None=@n17))
  NamedRerouteDeclaration(None=ComponentMask(None=VectorParameter() {parameterName=CameraVelocity, defaultValue=<rgba 0.5 0.25 0 0>}.RGB) {g=true, r=true}) {name=CamVel, nodeColor=<rgba 1 0.9609 0 1>}
  NamedRerouteDeclaration(None=Subtract(A=@n14.Inner, B=@n15)) {name=CornerPips, nodeColor=<rgba 0 0.0625 1 1>}
  NamedRerouteDeclaration(None=Subtract(A=@n16.Inner, B=@n17)) {name=CenterRing, nodeColor=<rgba 0.3828 1 0 1>}
  n18 = TextureCoordinate()
  n19 = Constant()
  n20 = MaterialFunctionCall(
   UVs=MaterialFunctionCall(
    UVs=MaterialFunctionCall(UVs=@n18, "Rotation Angle (0-1)"=@n19) {materialFunction=&CustomRotator}.Rotated Values
    Sides=Constant() {r=4}
   ) {materialFunction=&MF_UI_UVRadialMirror}.UVs
   Texture Scale=Constant() {r=0.15}
  ) {materialFunction=&ScaleUVsByCenter}
  n21 = ComponentMask(None=@n20.UVs) {g=true}
  n22 = ScalarParameter() {defaultValue=-1.2, parameterName=TriangleSpread}
  n23 = MaterialFunctionCall(UVs=AppendVector(A=ComponentMask(None=@n20.UVs) {r=true}, B=Add(A=@n21, B=@n22) {constB=-0.4}), "Rotation Angle (0-1)"=Divide() {constB=6}) {materialFunction=&CustomRotator}
  n24 = Constant() {r=3}
  NamedRerouteDeclaration(None=MaterialFunctionCall(DistanceField=MaterialFunctionCall!(UVs=@n23.Rotated Values, Radius, Sides=@n24, InnerAngle) {materialFunction=/FortUILibrary/MaterialFunctions/Shapes/SDFs/MF_SDF_RegularPolygon}.SDF) {materialFunction=&MF_UI_AntiAliasing}.Inner) {name=CornerTriangles, nodeColor=<rgba 0.0078 1 0 1>}
 ]
}
