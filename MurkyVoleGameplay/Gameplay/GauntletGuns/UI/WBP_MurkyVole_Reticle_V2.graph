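# Reticle extension widget for the MurkyVole Gauntlet Guns, showing overheat, missile ammo, spread and valid-target state.
# On construct it binds to HUD spread and the equipped weapon's overheat, missile-count and target-state delegates, which drive dynamic material parameters.
# Overheat cooldown dims the center reticle, and valid targets play a material animation forward or in reverse.
# Data-table values size the clustered missile display; camera-rotation effects on frame and missiles update each tick.
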WidgetBlueprint /MurkyVoleGameplay/Gameplay/GauntletGuns/UI/WBP_MurkyVole_Reticle_V2
animations [
 OverheatFadeOut {
  binding_count=0
  length_seconds=0.5
  object=&self:OverheatFadeOut
  sequence {bindings []}
 }
 ValidTargetForMissles {
  binding_count=1
  length_seconds=0.2
  object=&self:ValidTargetForMissles
  sequence {bindings [CenterRet {
     id=#1
     possessed_class=Image
     tracks [{
       type=MovieSceneWidgetMaterialTrack
       sections [{
         type=MovieSceneParameterSection
         channels [
          TargetInRangeSwitch {type=MovieSceneScriptingFloatChannel, keys [[0,0.25],[0.6667,0],[4,1]]}
          OuterPipOpacity {type=MovieSceneScriptingFloatChannel, keys [[2.6667,1],[3.3333,0],[4,1]]}
         ]
        }]
      }]
    }]}
 }
]
graphs [
 EventGraph {nodes [
   HandleMissileNumberChanged!() -> (OutputDelegate, then, OldMissileNumber:int) {is=title K2Node_CustomEvent}
   n1 = Get GauntletGuns() -> (GauntletGuns:B Athena Murky Vole Gauntlet Guns Object Reference)
   > OnRocketsRemainingChanged!(
    execute
    self
    RocketsRemaining:double="To Float (Integer)"(InInt=Get ActualMissileNumber!(self:Fort Weapon Ranged Dual Murky Vole Object Reference=@n1) -> (ActualMissileNumber:int=0))
   ) -> (then)
   Event Destruct()
   > Parent: Destruct()
   n2 = Create Event()
   n3 = Get FortHUDContext!() -> (ReturnValue:Fort HUDContext Object Reference) {is=title K2Node_GetBlueprintContextEx}
   > Unbind Event from On Spread Radius Changed(self:Fort HUDContext Object Reference=@n3, Delegate=@n2) {is=title K2Node_RemoveDelegate}
   > UnbindWeaponEvents()
   Event Construct()
   > Parent: Construct()
   > Bind Event to On Spread Radius Changed(self:Fort HUDContext Object Reference=@n3, Delegate=@n2) {is=title K2Node_AddDelegate}
   > SetAndBindToWeapon!(
    execute
    self
    NewWeapon:Fort Weapon Object Reference=GetEquippedWeapon!(self:Fort Weapon Reticle Extension Widget Base Object Reference) -> (ReturnValue:Fort Weapon Object Reference)
   ) -> (then)
   n4 = Get GauntletGuns()
   > Is Valid(InputObject=@n4)
   > OnOverheatValueChanged!(
    execute
    self
    CurrentValue:float=GetCurrentOverheatValue!(self:Fort Weapon Ranged Object Reference=@n4) -> (ReturnValue:float)
    MaxValue:float=GetOverheatingMaxValue!(self:Fort Weapon Ranged Object Reference=@n4) -> (ReturnValue:float)
    State:EFortWeaponOverheatState Enum=GetOverheatState!(self:Fort Weapon Ranged Object Reference=@n4) -> (ReturnValue:EFortWeaponOverheatState Enum=None)
    bIsInCooldown:bool=IsOverheated!(self:Fort Weapon Ranged Object Reference=@n4) -> (ReturnValue:bool)
   ) -> (then)
   n5 = Get GauntletGuns()
   > OnRocketsRemainingChanged(RocketsRemaining="To Float (Integer)"(InInt=Get ActualMissileNumber(self=@n5) -> (ActualMissileNumber=0)))
   n6 = Event Tick()
   > Parent: Tick!(execute, MyGeometry:Geometry Structure=@n6.MyGeometry, InDeltaTime:float=@n6.InDeltaTime) -> (then) {is=title K2Node_CallParentFunction}
   > Update!(
    execute
    self:Fort Widget Camera Rotation Effects Object Reference=Get CameraRotationEffects() -> (CameraRotationEffects:Fort Widget Camera Rotation Effects Object Reference)
    DeltaTime:float=@n6.InDeltaTime
   ) -> (then)
   Event OnInitialized()
   > Parent: OnInitialized() {is=title K2Node_CallParentFunction}
   > n7 = GetDynamicMaterial(self=Get Heat() -> (Heat:Image Object Reference))
   > Set OverheatDynamicMaterial(OverheatDynamicMaterial:Material Instance Dynamic Object Reference=@n7) -> (Output_Get:Material Instance Dynamic Object Reference)
   > n8 = GetDynamicMaterial(self=Get Missiles() -> (Missiles:Image Object Reference))
   > Set MissilesDynamicMaterial(MissilesDynamicMaterial:Material Instance Dynamic Object Reference=@n8) -> (Output_Get:Material Instance Dynamic Object Reference)
   > InitializeAmmoClusterVisuals()
   > InitializeCameraEffects()
  ]}
 InitializeAmmoClusterVisuals {nodes [
   InitializeAmmoClusterVisuals() {is=title K2Node_FunctionEntry}
   > Set ClusterSize(ClusterSize:float=GetValueAtLevel!(Input:"Scalable Float Structure (by ref)"=Get RowClusterSize() -> (RowClusterSize:Scalable Float Structure), Level:float=0) -> (ReturnValue:float)) -> (Output_Get:float)
   n1 = Get RowMaxAmmo() -> (RowMaxAmmo:Scalable Float Structure)
   > Set RowCount(RowCount:double=float / float(A=GetValueAtLevel(Input=@n1, Level=0), B=Get ClusterSize() -> (ClusterSize:float))) -> (Output_Get:double)
   n2 = Get MissilesDynamicMaterial() -> (MissilesDynamicMaterial:Material Instance Dynamic Object Reference)
   > SetScalarParameterValue(self=@n2, ParameterName=ClusterSize, Value=Get ClusterSize())
   > SetScalarParameterValue(self=@n2, ParameterName=ClusterCount, Value=Get RowCount() -> (RowCount:double))
   n3 = Get Missiles()
   n4 = Break Slate Brush!(SlateBrush:"Slate Brush Structure (by ref)"=Get Brush!(self:Image Object Reference=@n3) -> (Brush:Slate Brush Structure)) -> (
    TintColor:Slate Color Structure
    DrawAs:ESlateBrushDrawType Enum
    Tiling:ESlateBrushTileType Enum
    Mirroring:ESlateBrushMirrorType Enum
    ImageSize:Vector 2D Structure
    Margin:Margin Structure
    ResourceObject:Object Reference
    OutlineSettings:Slate Brush Outline Settings Structure
   ) {is=title K2Node_BreakStruct}
   n5 = float * float(A=Get RowCount(), B=42) {is=title K2Node_PromotableOperator}
   n6 = float * float(A=Get ClusterSize(), B=16)
   > SetBrush(
    self=@n3
    InBrush=Make Slate Brush!(
     TintColor:Slate Color Structure=@n4
     DrawAs:ESlateBrushDrawType Enum=@n4.DrawAs
     Tiling:ESlateBrushTileType Enum=@n4.Tiling
     Mirroring:ESlateBrushMirrorType Enum=@n4.Mirroring
     ImageSize:Vector 2D Structure=MakeVector2D(X=@n6, Y=@n5)
     Margin:Margin Structure=@n4.Margin
     ResourceObject:Object Reference=@n4.ResourceObject
     OutlineSettings:Slate Brush Outline Settings Structure=@n4.OutlineSettings
    ) -> (SlateBrush:Slate Brush Structure) {is=title K2Node_MakeStruct}
   )
  ]}
 InitializeCameraEffects {nodes [
   InitializeCameraEffects() {is=title K2Node_FunctionEntry}
   > n1 = Create UICamera Effects from Image List!(
    execute:Exec
    ImageList:Array of Image Object References=Make Array!("[0]":Image Object Reference=Get Frame() -> (Frame:Image Object Reference), "[1]":Image Object Reference=Get Missiles()) -> (Array:Array of Image Object References)
    PitchMaterial:Material Instance Dynamic Object Reference
   ) -> (then:Exec, CameraEffectsObject:"Fort Widget Camera Rotation Effects Object Reference (by ref)") {is=title K2Node_MacroInstance}
   > Set CameraRotationEffects(CameraRotationEffects:Fort Widget Camera Rotation Effects Object Reference=@n1) -> (Output_Get:Fort Widget Camera Rotation Effects Object Reference)
  ]}
 OnIsValidTargetChanged {nodes [
   n1 = OnIsValidTargetChanged!() -> (then, bIsValidTarget:bool=~) {is=title K2Node_FunctionEntry}
   > n2 = Branch(Condition=@n1)
   n3 = Get ValidTargetForMissles() -> (ValidTargetForMissles:Widget Animation Object Reference)
   > PlayAnimationForward(InAnimation=@n3)
   PlayAnimationReverse(execute=@n2.else, InAnimation=@n3)
  ]}
 OnOverheatValueChanged {nodes [
   n1 = OnOverheatValueChanged!() -> (then, CurrentValue:float=~, MaxValue:float=~, State:EFortWeaponOverheatState Enum=None, bIsInCooldown:bool=~) {is=title K2Node_FunctionEntry}
   > n2 = Set PercentFill(PercentFill:double=float / float(A=@n1, B=@n1.MaxValue)) -> (Output_Get:double)
   > SetScalarParameterValue(
    self=Get OverheatDynamicMaterial() -> (OverheatDynamicMaterial:Material Instance Dynamic Object Reference)
    ParameterName=OverHeat
    Value=MapRangeClamped(Value=@n2, InRangeA=0, InRangeB=0.96)
   )
   > n3 = Branch(Condition=@n1.bIsInCooldown)
   n4 = Get OverheatDynamicMaterial()
   > SetScalarParameterValue(self=@n4, ParameterName=Overheated?, Value=1)
   n5 = Get CenterRet() -> (CenterRet:Image Object Reference)
   > Set RenderOpacity!(execute, RenderOpacity:float=0.25, self:Widget Object Reference=@n5) -> (then, Output_Get:float)
   SetScalarParameterValue(execute=@n3.else, self=@n4, ParameterName=Overheated?, Value=0)
   > Set RenderOpacity(RenderOpacity=1, self=@n5)
  ]}
 OnRocketsRemainingChanged {nodes [
   n1 = OnRocketsRemainingChanged!() -> (then, RocketsRemaining:double=~) {is=title K2Node_FunctionEntry}
   n2 = Get Row_MaxSegments() -> (Row_MaxSegments:Scalable Float Structure)
   > SetScalarParameterValue(
    self=Get MissilesDynamicMaterial()
    ParameterName=RocketsRemaining
    Value="Clamp (Float)"(Value=@n1, Min=0, Max=GetValueAtLevel(Input=@n2, Level=0))
   )
  ]}
 OnSpreadRadiusChanged {nodes [
   n1 = OnSpreadRadiusChanged!() -> (then, SpreadRadius:float=~) {is=title K2Node_FunctionEntry}
   > n2 = GetDynamicMaterial(self=Get CenterRet())
   > SetScalarParameterValue(self=@n2, ParameterName=Spread, Value=@n1)
  ]}
 SetAndBindToWeapon {nodes [
   n1 = SetAndBindToWeapon!() -> (then, NewWeapon:Fort Weapon Object Reference) {is=title K2Node_FunctionEntry}
   > UnbindWeaponEvents()
   > n2 = Cast To B_Athena_MurkyVole_GauntletGuns!(execute, Object=@n1) -> (then, CastFailed, AsB Athena Murky Vole Gauntlet Guns:B Athena Murky Vole Gauntlet Guns Object Reference) {is=title K2Node_DynamicCast}
   > n3 = Set GauntletGuns(GauntletGuns:B Athena Murky Vole Gauntlet Guns Object Reference=@n2) -> (Output_Get:B Athena Murky Vole Gauntlet Guns Object Reference)
   > Bind Event to On Overheat Value Changed(self:Fort Weapon Ranged Object Reference=@n3, Delegate=Create Event()) {is=title K2Node_AddDelegate}
   > Bind Event to On Missile Number Changed(self:Fort Weapon Ranged Dual Murky Vole Object Reference=@n3, Delegate=Create Event()) {is=title K2Node_AddDelegate}
   > n4 = Set TargetStateComponent(TargetStateComponent:Fort Weapon Component Target State Object Reference=GetComponentByClass!(self:Actor Object Reference=@n3, ComponentClass:Actor Component Class Reference=/Script/FortniteGame.FortWeaponComponent_TargetState) -> (ReturnValue:Fort Weapon Component Target State Object Reference)) -> (Output_Get:Fort Weapon Component Target State Object Reference)
   > Is Valid(InputObject=@n4)
   > Bind Event to On Is Valid Target Changed(self:Fort Weapon Component Target State Object Reference=@n4, Delegate=Create Event()) {is=title K2Node_AddDelegate}
   Set GauntletGuns(execute=@n2.CastFailed)
  ]}
 UnbindWeaponEvents {nodes [
   UnbindWeaponEvents() {is=title K2Node_FunctionEntry}
   > n1 = Sequence()
   n2 = Get GauntletGuns()
   > Is Valid(InputObject=@n2)
   > Unbind Event from On Overheat Value Changed(self:Fort Weapon Ranged Object Reference=@n2, Delegate=Create Event()) {is=title K2Node_RemoveDelegate}
   > Unbind Event from On Missile Number Changed(self:Fort Weapon Ranged Dual Murky Vole Object Reference=@n2, Delegate=Create Event()) {is=title K2Node_RemoveDelegate}
   n3 = Get TargetStateComponent() -> (TargetStateComponent:Fort Weapon Component Target State Object Reference)
   Is Valid(exec=@n1.then_1, InputObject=@n3)
   > Unbind Event from On Is Valid Target Changed(self:Fort Weapon Component Target State Object Reference=@n3, Delegate=Create Event()) {is=title K2Node_RemoveDelegate}
  ]}
]
variables [
 CameraRotationEffects {type=FortWidgetCameraRotationEffects, default=null}
 GauntletGuns {type=B_Athena_MurkyVole_GauntletGuns_C, default=null}
 MissilesDynamicMaterial {type=MaterialInstanceDynamic}
 OverheatDynamicMaterial {type=MaterialInstanceDynamic}
 Row_MaxSegments {
  type=ScalableFloat
  default {
   registryType=None
   value=1
   curve {curveTable=/MurkyVoleGameplay/DataTables/MurkyVoleGameplayGameData, rowName=Default.MurkyVole.MissileSwarm.Ammo.Max}
  }
 }
 Row_PawnLock_HalfHeight {
  type=ScalableFloat
  default {
   registryType=None
   value=1
   curve {curveTable=&MurkyVoleGameplayGameData, rowName=Default.MurkyVole.MissileSwarm.PawnLock.HalfWidth}
  }
 }
 Row_PawnLock_HalfWidth {
  type=ScalableFloat
  default {
   registryType=None
   value=1
   curve {curveTable=&MurkyVoleGameplayGameData, rowName=Default.MurkyVole.MissileSwarm.PawnLock.HalfWidth}
  }
 }
 Row_PawnLock_Range {
  type=ScalableFloat
  default {
   registryType=None
   value=1
   curve {curveTable=&MurkyVoleGameplayGameData, rowName=Default.MurkyVole.MissileSwarm.PawnLock.Range}
  }
 }
 RowClusterSize {
  type=ScalableFloat
  default {
   registryType=None
   value=1
   curve {curveTable=&MurkyVoleGameplayGameData, rowName=Default.MurkyVole.MissileSwarm.Cluster.Size.Min}
  }
 }
 RowMaxAmmo {
  type=ScalableFloat
  default {
   registryType=None
   value=1
   curve {curveTable=&MurkyVoleGameplayGameData, rowName=Default.MurkyVole.MissileSwarm.Ammo.Max}
  }
 }
 TargetStateComponent {type=FortWeaponComponent_TargetState, default=null}
]
widgets {
 info {
  inheritedWidgetCount=1
  parentClass=/Script/FortniteUI.FortWeaponReticleExtensionWidgetBase
  rootWidgetClass=&GridPanel
  widgetCount=6
 }
 widgets [
  GridPanel_1 {widgetClassPath=&GridPanel, widget=GridPanel_1}
  CenterRet {
   widgetClassPath=&Image
   bIsVariable=true
   parent=GridPanel_1
   widget=CenterRet
   properties {brush {
     bIsDynamicallyLoaded=false
     drawAs=Image
     imageType=FullColor
     mirroring=NoMirror
     resourceName=None
     resourceObject=&/M_UI_MurkyVole_GauntletGuns_Reticle_Center
     tiling=NoTile
     imageSize=<256 256>
     margin=<ltrb 0 0 0 0>
     outlineSettings {
      bUseBrushTransparency=false
      roundingType=HalfHeightRadius
      width=0
      color {colorUseRule=UseColor_Specified, specifiedColor=<rgba 0 0 0 0>}
      cornerRadii=<0 0 0 0>
     }
     tintColor {colorUseRule=UseColor_Specified, specifiedColor=<rgba 1 1 1 1>}
     uVRegion {bIsValid=false, max=<0 0>, min=<0 0>}
    }}
   slot {class=GridSlot, properties {horizontalAlignment=HAlign_Center, verticalAlignment=VAlign_Center}}
  }
  Heat {
   widgetClassPath=&Image
   bIsVariable=true
   parent=GridPanel_1
   widget=Heat
   properties {brush {
     bIsDynamicallyLoaded=false
     drawAs=Image
     imageType=FullColor
     mirroring=NoMirror
     resourceName=None
     resourceObject=/MurkyVoleGameplay/Gameplay/GauntletGuns/UI/V2/MI_UI_MurkyVole_Overheat
     tiling=NoTile
     imageSize=<34 130>
     margin=<ltrb 0 0 0 0>
     outlineSettings {
      bUseBrushTransparency=false
      roundingType=HalfHeightRadius
      width=0
      color {colorUseRule=UseColor_Specified, specifiedColor=<rgba 0 0 0 0>}
      cornerRadii=<0 0 0 0>
     }
     tintColor {colorUseRule=UseColor_Specified, specifiedColor=<rgba 1 1 1 1>}
     uVRegion {bIsValid=false, max=<0 0>, min=<0 0>}
    }}
   slot {class=GridSlot, properties {horizontalAlignment=HAlign_Center, verticalAlignment=VAlign_Center, nudge=<95 0>}}
  }
  Frame {
   widgetClassPath=&Image
   bIsVariable=true
   parent=GridPanel_1
   widget=Frame
   properties {brush {
     bIsDynamicallyLoaded=false
     drawAs=Image
     imageType=FullColor
     mirroring=NoMirror
     resourceName=None
     resourceObject=/MurkyVoleGameplay/Gameplay/GauntletGuns/UI/V2/MI_UI_MurkyVole_Gauntlets_Frame
     tiling=NoTile
     imageSize=<512 256>
     margin=<ltrb 0 0 0 0>
     outlineSettings {
      bUseBrushTransparency=false
      roundingType=HalfHeightRadius
      width=0
      color {colorUseRule=UseColor_Specified, specifiedColor=<rgba 0 0 0 0>}
      cornerRadii=<0 0 0 0>
     }
     tintColor {colorUseRule=UseColor_Specified, specifiedColor=<rgba 1 1 1 0.5>}
     uVRegion {bIsValid=false, max=<0 0>, min=<0 0>}
    }}
   slot {class=GridSlot, properties {horizontalAlignment=HAlign_Center, verticalAlignment=VAlign_Center}}
  }
  Missiles {
   widgetClassPath=&Image
   bIsVariable=true
   parent=GridPanel_1
   widget=Missiles
   properties {
    brush {
     bIsDynamicallyLoaded=false
     drawAs=Image
     imageType=FullColor
     mirroring=NoMirror
     resourceName=None
     resourceObject=/MurkyVoleGameplay/Gameplay/GauntletGuns/UI/V2/MI_UI_MurkyVole_Gauntlets_Missiles
     tiling=NoTile
     imageSize=<32 126>
     margin=<ltrb 0 0 0 0>
     outlineSettings {
      bUseBrushTransparency=false
      roundingType=HalfHeightRadius
      width=0
      color {colorUseRule=UseColor_Specified, specifiedColor=<rgba 0 0 0 0>}
      cornerRadii=<0 0 0 0>
     }
     tintColor {colorUseRule=UseColor_Specified, specifiedColor=<rgba 1 1 1 1>}
     uVRegion {bIsValid=false, max=<0 0>, min=<0 0>}
    }
    renderTransformPivot=<1 0.5>
   }
   slot {class=GridSlot, properties {horizontalAlignment=HAlign_Center, verticalAlignment=VAlign_Center, nudge=<170 0>}}
  }
  TutorialHighlight {widgetClassPath=&TutorialHighlight}
 ]
}
